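# Physics asset for the right-side shorts of the male medium BlueGlaze body cosmetic, simulating the garment's dynamic bones.
# Capsule bodies cover four shorts bones (front, back, inner, outer), their master bone, and the right thigh and calf twist bones.
# The four shorts bones are constrained to the master with narrow swing and zero twist limits, while the master swings and twists freely on the thigh.
# The calf capsule has no constraint.
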PhysicsAsset /BRCosmetics/Characters/Player/Male/Medium/Bodies/M_MED_BlueGlaze/Meshes/M_MED_BlueGlaze_R_Physics
bodies {
 calf_twist_02_r [{shapeType=Capsule, length=30, radius=5, shapeName=thigh_twist_01_l_capsule, rotation=<rot 90 0 0>}]
 dyn_R_shorts_bk [{shapeType=Capsule, length=4, radius=4, shapeName=dyn_R_shorts_bk_capsule, center=<-3.9506 3.821 0>}]
 dyn_R_shorts_fr [{shapeType=Capsule, length=5, radius=5, shapeName=dyn_L_shorts_fr_capsule, center=<-6.4365 -1.7735 0>, rotation=<rot 90 0 0>}]
 dyn_R_shorts_in [{shapeType=Capsule, length=5, radius=5, shapeName=dyn_L_shorts_out_capsule, center=<-6.4365 0 3.1974>, rotation=<rot 90 0 0>}]
 dyn_R_shorts_out [{shapeType=Capsule, length=5, radius=5, shapeName=dyn_L_shorts_in_capsule, center=<-6.4365 0 -3.2927>, rotation=<rot 90 0 0>}]
 master_R_dyn_shorts_main [{shapeType=Capsule, length=5, radius=5, shapeName=master_L_dyn_shorts_main_capsule, rotation=<rot 90 0 0>}]
 thigh_twist_01_r [{shapeType=Capsule, length=28.25, radius=6.75, shapeName=thigh_twist_01_l_capsule, rotation=<rot 90 0 0>}]
}
constraints [
 {bone1Name=dyn_R_shorts_fr, bone2Name=master_R_dyn_shorts_main, swing1LimitDegrees=5, swing2LimitDegrees=5, twistLimitDegrees=0}
 {bone1Name=dyn_R_shorts_bk, bone2Name=master_R_dyn_shorts_main, swing1LimitDegrees=5, swing2LimitDegrees=5, twistLimitDegrees=0}
 {bone1Name=dyn_R_shorts_in, bone2Name=master_R_dyn_shorts_main, swing1LimitDegrees=5, swing2LimitDegrees=5, twistLimitDegrees=0}
 {bone1Name=dyn_R_shorts_out, bone2Name=master_R_dyn_shorts_main, swing1LimitDegrees=5, swing2LimitDegrees=5, twistLimitDegrees=0}
 {
  bone1Name=master_R_dyn_shorts_main
  bone2Name=thigh_twist_01_r
  swing1LimitDegrees=45
  swing1Motion=Free
  swing2LimitDegrees=45
  swing2Motion=Free
  twistLimitDegrees=45
  twistMotion=Free
 }
]
