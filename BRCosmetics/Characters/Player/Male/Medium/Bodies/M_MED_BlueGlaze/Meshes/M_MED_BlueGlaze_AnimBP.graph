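# Post-copy AnimBP for the BlueGlaze male medium body cosmetic, adding dynamic shorts and shoulder strap motion to the pose copied from the attached parent mesh.
# A deform Control Rig feeds per-thigh RigidBody simulations, then additive shorts bone offsets eased by crouch and jump alphas.
# A weighted constraint drives the right strap helper before a final Control Rig receives sprint, skydive, parachute and ADS states.
# The EventGraph refreshes those states from movement flags each update.
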AnimBlueprint /BRCosmetics/Characters/Player/Male/Medium/Bodies/M_MED_BlueGlaze/Meshes/M_MED_BlueGlaze_AnimBP
target_skeleton=&/M_MED_BlueGlaze_Skeleton
graphs [
 AnimGraph {nodes [
   n1 = Control Rig(Source=Copy Pose From Mesh() {bUseAttachedParent=true}) {
    alphaBoolBlend {blendOption=Linear}
    controlRigAssetReference {blueprintRigClass=&/M_MED_DeformRig_BlueGlaze_ControlRig.M_MED_DeformRig_BlueGlaze_ControlRig_C}
   }
   n2 = RigidBody(ComponentPose=Local To Component(LocalPose=@n1)) {
    bClampLinearTranslationLimitToRefPose=true
    bOverrideWorldGravity=true
    overridePhysicsAsset=&/M_MED_BlueGlaze_R_Physics
    simulationSpace=BaseBoneSpace
    alphaBoolBlend {blendOption=Linear}
    baseBoneRef {boneName=thigh_r}
    componentAppliedLinearAccClamp=<3000 3000 3000>
    componentLinearAccScale=<0.2 0.1 0.1>
    componentLinearVelScale=<0.2 0.1 0.1>
    overrideWorldGravity {z=-980}
    simSpaceSettings {maxAngularAcceleration=1000, maxAngularVelocity=3, maxLinearAcceleration=250, maxLinearVelocity=500, velocityScaleZ=0, worldAlpha=1}
   }
   n3 = RigidBody(ComponentPose=@n2) {
    bClampLinearTranslationLimitToRefPose=true
    bOverrideWorldGravity=true
    overridePhysicsAsset=&/M_MED_BlueGlaze_L_Physics
    simulationSpace=BaseBoneSpace
    alphaBoolBlend {blendOption=Linear}
    baseBoneRef {boneName=thigh_l}
    componentAppliedLinearAccClamp=<3000 3000 3000>
    componentLinearAccScale=<0.2 0.1 0.1>
    componentLinearVelScale=<0.2 0.1 0.1>
    overrideWorldGravity {z=-980}
    simSpaceSettings {maxAngularAcceleration=1000, maxAngularVelocity=3, maxLinearAcceleration=250, maxLinearVelocity=500, velocityScaleZ=0, worldAlpha=1}
   }
   n4 = Get Is Crouching() -> (Is Crouching:double)
   n5 = "Transform (Modify) Bone - Bone: master_L_dyn_shorts_main"(ComponentPose=@n3, Alpha=@n4, Translation="-7.2,1,0", Rotation="0,0,0", Scale="1,1,1") {
    rotationMode=BMM_Additive
    rotationSpace=BCS_BoneSpace
    translationMode=BMM_Additive
    translationSpace=BCS_BoneSpace
    alphaBoolBlend {blendOption=Linear}
    alphaScaleBiasClamp {bInterpResult=true, interpSpeedIncreasing=15}
    boneToModify {boneName=master_L_dyn_shorts_main}
    translation=<-7.2 1>
   }
   n6 = "Transform (Modify) Bone - Bone: master_R_dyn_shorts_main"(ComponentPose=@n5, Alpha=@n4, Translation="2.6,-2.9,0", Rotation="0,-20,0", Scale="1,1,1") {
    rotationMode=BMM_Additive
    rotationSpace=BCS_BoneSpace
    translationMode=BMM_Additive
    translationSpace=BCS_BoneSpace
    alphaBoolBlend {blendOption=Linear}
    alphaScaleBiasClamp {bInterpResult=true, interpSpeedIncreasing=15}
    boneToModify {boneName=master_R_dyn_shorts_main}
    rotation {yaw=-20}
    translation=<2.6 -2.9>
   }
   n7 = Get Is Jumping() -> (Is Jumping:double)
   n8 = "Transform (Modify) Bone - Bone: master_R_dyn_shorts_main"(ComponentPose=@n6, Alpha=@n7, Translation="8,0,0", Rotation="0,0,0", Scale="1,1,1") {
    rotationMode=BMM_Additive
    rotationSpace=BCS_BoneSpace
    translationMode=BMM_Additive
    translationSpace=BCS_BoneSpace
    alphaBoolBlend {blendOption=Linear}
    alphaScaleBiasClamp {bInterpResult=true, interpSpeedIncreasing=15}
    boneToModify {boneName=master_R_dyn_shorts_main}
    translation {x=8}
   }
   n9 = "Transform (Modify) Bone - Bone: master_L_dyn_shorts_main"(ComponentPose=@n8, Alpha=@n7, Translation="-7,0,0", Rotation="0,0,0", Scale="1,1,1") {
    rotationMode=BMM_Additive
    rotationSpace=BCS_BoneSpace
    translationMode=BMM_Additive
    translationSpace=BCS_BoneSpace
    alphaBoolBlend {blendOption=Linear}
    alphaScaleBiasClamp {bInterpResult=true, interpSpeedIncreasing=15}
    boneToModify {boneName=master_L_dyn_shorts_main}
    translation {x=-7}
   }
   n10 = "Transform (Modify) Bone - Bone: dyn_L_shorts_main"(ComponentPose=@n9, Alpha=@n7, Translation="0,4,0", Rotation="0,-20,0", Scale="1,1,1") {
    rotationMode=BMM_Additive
    rotationSpace=BCS_BoneSpace
    translationMode=BMM_Additive
    translationSpace=BCS_BoneSpace
    alphaBoolBlend {blendOption=Linear}
    alphaScaleBiasClamp {bInterpResult=true, interpSpeedDecreasing=1}
    boneToModify {boneName=dyn_L_shorts_main}
    rotation {yaw=-20}
    translation {y=4}
   }
   n11 = "Transform (Modify) Bone - Bone: dyn_R_shorts_main"(ComponentPose=@n10, Alpha=@n7, Translation="0,-4,0", Rotation="0,-20,0", Scale="1,1,1") {
    rotationMode=BMM_Additive
    rotationSpace=BCS_BoneSpace
    translationMode=BMM_Additive
    translationSpace=BCS_BoneSpace
    alphaBoolBlend {blendOption=Linear}
    alphaScaleBiasClamp {bInterpResult=true, interpSpeedDecreasing=1}
    boneToModify {boneName=dyn_R_shorts_main}
    rotation {yaw=-20}
    translation {y=-4}
   }
   n12 = "Constraint - R_shoulder_strap_helper_jnt (PPPPPP)"!(
    ComponentPose=@n11
    Alpha=1
    ConstraintWeights_0=0.0318
    ConstraintWeights_1:float=0.0658
    ConstraintWeights_2:float=0.5689
    ConstraintWeights_3:float=0.2632
    ConstraintWeights_4:float=0.0519
    ConstraintWeights_5:float=0.0183
   ) -> (Pose) {
    alphaBoolBlend {blendOption=Linear}
    boneToModify {boneName=R_shoulder_strap_helper_jnt}
    constraintSetup [
     {targetBone {boneName=spine_04}}
     {targetBone {boneName=spine_05}}
     {targetBone {boneName=clavicle_r}}
     {targetBone {boneName=upperarm_twist_01_r}}
     {targetBone {boneName=upperarm_twist_02_r}}
     {targetBone {boneName=neck_01}}
    ]
    constraintWeights [1,1,1,1,1,1]
   }
   n13 = Control Rig!(
    Source=Component To Local(ComponentPose=@n12)
    Alpha
    Is Tac Sprinting:bool=Get Is Tac Sprinting() -> (Is Tac Sprinting:bool)
    Is Skydive Floating:double=Get Is Skydive Floating() -> (Is Skydive Floating:double)
    Is Parachute Open:double=Get Is Parachute Open() -> (Is Parachute Open:double)
    Is ADS:bool=Get Is ADS() -> (Is ADS:bool)
    Is Skydiving:bool=Get Is Skydiving() -> (Is Skydiving:bool)
   ) -> (Pose) {
    alphaBoolBlend {blendOption=Linear}
    alphaScaleBiasClamp {interpSpeedDecreasing=2}
    controlRigAssetReference {blueprintRigClass=&/M_MED_BlueGlaze_CtrlRig.M_MED_BlueGlaze_CtrlRig_C}
   }
   Output Pose(Result=@n13) {layerGroup=None}
  ]}
 EventGraph {nodes [
   Event BlueprintUpdateAnimation()
   > Set Is Crouching(Is Crouching:double="To Float (Boolean)"(InBool=Get bIsCrouching())) -> (Output_Get:double)
   n1 = NOT Boolean(A=Get bIsParachuteOpened())
   n2 = OR Boolean(A=Get bIsJumping(), B=Get bIsFalling())
   > Set Is Jumping(Is Jumping:double="To Float (Boolean)"(InBool=AND Boolean(A=@n1, B=@n2))) -> (Output_Get:double)
   > Set Is Moving(Is Moving:bool=Get bIsMoving2D()) -> (Output_Get:bool)
   n3 = Get bIsMoving2D()
   n4 = Get bIsCrouching()
   > Set Is Crouch Moving(Is Crouch Moving:double="To Float (Boolean)"(InBool=AND Boolean(A=@n4, B=@n3))) -> (Output_Get:double)
   > Set Is Tac Sprinting(Is Tac Sprinting:bool=Get bIsTacticalSprinting()) -> (Output_Get:bool)
   > Set Is Skydive Floating(Is Skydive Floating:double="To Float (Boolean)"(InBool=Get bIsSkydiveFloating())) -> (Output_Get:double)
   > Set Is Parachute Open(Is Parachute Open:double="To Float (Boolean)"(InBool=Get bIsParachuteOpened())) -> (Output_Get:double)
   > Set Is ADS(Is ADS:bool=Get bIsTargeting()) -> (Output_Get:bool)
   > Set Is Skydiving(Is Skydiving:bool=OR Boolean!(A=Get bIsParachuteOpened(), B=Get bIsSkyDiving(), C:bool=Get bIsSkydiveFloating()) -> (ReturnValue)) -> (Output_Get:bool)
   > Set Is Sprinting(Is Sprinting:bool=Get bIsSprinting()) -> (Output_Get:bool)
   Get bIsSprinting()
  ]}
]
variables [
 Is ADS {type=boolean}
 Is Crouch Moving {type=number}
 Is Crouching {type=number}
 Is Jumping {type=number}
 Is Moving {type=boolean}
 Is Parachute Open {type=number}
 Is Skydive Floating {type=number}
 Is Skydiving {type=boolean}
 Is Sprinting {type=boolean}
 Is Tac Sprinting {type=boolean}
]
