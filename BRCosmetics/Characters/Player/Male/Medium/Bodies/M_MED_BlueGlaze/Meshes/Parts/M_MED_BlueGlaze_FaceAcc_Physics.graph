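# Physics asset for the BlueGlaze medium male player's face accessory mesh, simulating a pair of headphones with small capsules on left and right joint chains.
# Each chain hangs from a master joint constrained to a large spine_05 torso capsule with free rotation.
# Most chain links have narrow swing and almost no twist.
# The fourth left joint has free-rotation constraints to both third joints, linking the sides into a closed loop across spine_05.
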PhysicsAsset /BRCosmetics/Characters/Player/Male/Medium/Bodies/M_MED_BlueGlaze/Meshes/Parts/M_MED_BlueGlaze_FaceAcc_Physics
bodies {
 L_headphones_master_jnt [{shapeType=Capsule, length=1, radius=1, shapeName=L_headphones_master_jnt_capsule}]
 R_headphones_master_jnt [{shapeType=Capsule, length=1, radius=1, shapeName=R_headphones_master_jnt_capsule}]
 dyn_L_headphones_1_jnt [{shapeType=Capsule, length=2.5, radius=2, shapeName=dyn_L_headphones_1_jnt_capsule, center=<-2.5038 0 0.2891>, rotation=<rot -90 0 0.0004>}]
 dyn_L_headphones_2_jnt [{
   shapeType=Capsule
   length=2
   radius=2
   shapeName=dyn_L_headphones_2_jnt_capsule
   center=<-1.9149 0 0.2406>
   rotation=<rot -80 0.0005 -0.0001>
  }]
 dyn_L_headphones_3_jnt [{
   shapeType=Capsule
   length=1.75
   radius=2
   shapeName=dyn_L_headphones_3_jnt_capsule
   center=<-2.2314 0 0.1361>
   rotation=<rot -90 166.7373 -166.7372>
  }]
 dyn_L_headphones_4_jnt [{
   shapeType=Capsule
   length=1.75
   radius=2
   shapeName=dyn_L_headphones_3_jnt_capsule
   center=<-2.2314 0 0.1361>
   rotation=<rot -90 166.7373 -166.7372>
  }]
 dyn_R_headphones_1_jnt [{shapeType=Capsule, length=2.75, radius=2, shapeName=dyn_R_headphones_1_jnt_capsule, center=<3.0623 0 0.0186>, rotation=<rot -80 0.0005 0>}]
 dyn_R_headphones_2_jnt [{
   shapeType=Capsule
   length=2.75
   radius=2
   shapeName=dyn_R_headphones_2_jnt_capsule
   center=<3.0623 0 0.0186>
   rotation=<rot -80 0.0005 -0.0001>
  }]
 dyn_R_headphones_3_jnt [{
   shapeType=Capsule
   length=3.5
   radius=2
   shapeName=dyn_R_headphones_3_jnt_capsule
   center=<3.5651 0 0.0186>
   rotation=<rot -90 179.9736 -179.9735>
  }]
 spine_05 [{shapeType=Capsule, length=19.5, radius=11.75, shapeName=spine_05_capsule, center=<0.3959 -2.0627 0>}]
}
constraints [
 {bone1Name=dyn_L_headphones_1_jnt, bone2Name=L_headphones_master_jnt, swing1LimitDegrees=15, swing2LimitDegrees=15, twistLimitDegrees=1}
 {bone1Name=dyn_L_headphones_2_jnt, bone2Name=dyn_L_headphones_1_jnt, swing1LimitDegrees=15, swing2LimitDegrees=15, twistLimitDegrees=1}
 {bone1Name=dyn_L_headphones_3_jnt, bone2Name=dyn_L_headphones_2_jnt, swing1LimitDegrees=15, swing2LimitDegrees=15, twistLimitDegrees=1}
 {bone1Name=dyn_R_headphones_1_jnt, bone2Name=R_headphones_master_jnt, swing1LimitDegrees=15, swing2LimitDegrees=15, twistLimitDegrees=1}
 {bone1Name=dyn_R_headphones_2_jnt, bone2Name=dyn_R_headphones_1_jnt, swing1LimitDegrees=15, swing2LimitDegrees=15, twistLimitDegrees=1}
 {bone1Name=dyn_R_headphones_3_jnt, bone2Name=dyn_R_headphones_2_jnt, swing1LimitDegrees=15, swing2LimitDegrees=15, twistLimitDegrees=1}
 {
  bone1Name=R_headphones_master_jnt
  bone2Name=spine_05
  swing1LimitDegrees=45
  swing1Motion=Free
  swing2LimitDegrees=45
  swing2Motion=Free
  twistLimitDegrees=45
  twistMotion=Free
 }
 {
  bone1Name=L_headphones_master_jnt
  bone2Name=spine_05
  swing1LimitDegrees=45
  swing1Motion=Free
  swing2LimitDegrees=45
  swing2Motion=Free
  twistLimitDegrees=45
  twistMotion=Free
 }
 {
  bone1Name=dyn_L_headphones_4_jnt
  bone2Name=dyn_L_headphones_3_jnt
  swing1LimitDegrees=45
  swing1Motion=Free
  swing2LimitDegrees=45
  swing2Motion=Free
  twistLimitDegrees=45
  twistMotion=Free
 }
 {
  bone1Name=dyn_L_headphones_4_jnt
  bone2Name=dyn_R_headphones_3_jnt
  swing1LimitDegrees=45
  swing1Motion=Free
  swing2LimitDegrees=45
  swing2Motion=Free
  twistLimitDegrees=45
  twistMotion=Free
 }
]
